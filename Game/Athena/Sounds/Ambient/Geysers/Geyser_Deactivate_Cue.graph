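# Sound cue named for geyser deactivation in Athena ambient audio, currently an empty shell with no root node or sound nodes.
# With no wave player in the graph, it has nothing to play.
# It still carries playback settings: the geyser activation attenuation asset, the Athena ambient sound class, a raised volume multiplier and a maximum audible distance.
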SoundCue /Game/Athena/Sounds/Ambient/Geysers/Geyser_Deactivate_Cue
root=null
nodes []
properties {
 attenuationSettings=&/Geyser_Activate_Att
 maxDistance=10000
 soundClassObject=&Fort_Ambient_Athena
 volumeMultiplier=1.559
}
