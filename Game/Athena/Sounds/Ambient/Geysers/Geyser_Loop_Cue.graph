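# Looping ambient sound cue for geysers, in the Athena ambient sound class with a concurrency limit.
# A close-range loop with listener focus and no reverb send is mixed with a longer-range, occluded "activated" loop, each using its own distance filtering.
# The activated loop fades in with the Intensity parameter at reduced mix volume; the other crossfade and inner mixer inputs are empty.
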SoundCue /Game/Athena/Sounds/Ambient/Geysers/Geyser_Loop_Cue
root=Mixer_1
nodes [
 {
  id=Attenuation_0
  children [Mixer_0]
  properties {
   bOverrideAttenuation=true
   attenuationOverrides {
    bAttenuateWithLPF=true
    bEnableOcclusion=true
    distanceAlgorithm=NaturalSound
    falloffDistance=9000
    hPFFrequencyAtMax=100
    lPFFrequencyAtMax=808
    lPFRadiusMax=9000
    lPFRadiusMin=500
    occlusionLowPassFilterFrequency=3000
    occlusionTraceChannel=ECC_GameTraceChannel11
    occlusionVolumeAttenuation=0.5
    reverbDistanceMax=5000
    stereoSpread=1000
    attenuationShapeExtents {x=0}
    customAttenuationCurve {editorCurveData {keys [<at 0 1>,<at 1 0>]}}
   }
  }
 }
 {
  id=Attenuation_1
  children [WavePlayer_0]
  properties {
   bOverrideAttenuation=true
   attenuationOverrides {
    bAttenuateWithLPF=true
    bEnableFocusInterpolation=true
    bEnableListenerFocus=true
    bEnableReverbSend=false
    distanceAlgorithm=NaturalSound
    falloffDistance=4500
    focusReleaseInterpSpeed=0.5
    hPFFrequencyAtMax=100
    lPFFrequencyAtMax=800
    lPFRadiusMax=4500
    lPFRadiusMin=500
    nonFocusVolumeAttenuation=0.5
    stereoSpread=500
    attenuationShapeExtents {x=1000}
   }
  }
 }
 {id=Mixer_0, children [null, ParamCrossFade_0], properties {inputVolume [1,0.25]}}
 {id=Mixer_1, children [Attenuation_1, Attenuation_0], properties {inputVolume [1,1]}}
 {
  id=ParamCrossFade_0
  children [null,WavePlayer_2]
  properties {
   paramName=Intensity
   crossFadeInput [
    {fadeInDistanceEnd=0, fadeInDistanceStart=0, fadeOutDistanceEnd=1, fadeOutDistanceStart=0, volume=1}
    {fadeInDistanceEnd=1, fadeInDistanceStart=0, fadeOutDistanceEnd=100, fadeOutDistanceStart=100, volume=1}
   ]
  }
 }
 {id=WavePlayer_0, properties {bLooping=true, soundWaveAssetPtr=&/Geyser_AmbientLoop_Close_01}}
 {id=WavePlayer_2, properties {bLooping=true, soundWaveAssetPtr=&/Geyser_AmbientLoop_Activated_01}}
]
properties {
 bOverrideConcurrency=true
 maxDistance=9000
 soundClassObject=&Fort_Ambient_Athena
 volumeMultiplier=1.559
 attenuationOverrides {attenuationShapeExtents {x=0}}
 concurrencyOverrides {maxCount=5}
}
